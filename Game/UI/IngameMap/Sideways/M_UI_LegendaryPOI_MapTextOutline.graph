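# UI material that, per its name, colors the text outline of legendary POI labels on the in-game map with animated blue bands.
# Normalized UI UVs pan along X, scale by SegmentCount and rotate slightly, giving a moving, tilted gradient.
# MF_UI_PingPong folds the X component into repeating bands, which at half strength blend from dark navy toward a brighter blue.
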Material /Game/UI/IngameMap/Sideways/M_UI_LegendaryPOI_MapTextOutline
graph {
 properties {bAutomaticallySetUsageInEditor=false, bUsedWithStaticMesh=false, materialDomain=MD_UI, usedShadingModels=Default Lit}
 nodes [
  n1 = Panner(Coordinate=MaterialFunctionCall() {materialFunction=&GetUserInterfaceUV}.Normalized UV) {speedX=-0.2}
  n2 = ScalarParameter() {defaultValue=1.5, parameterName=SegmentCount}
  n3 = ComponentMask(None=MaterialFunctionCall(UVs=Multiply(A=@n1, B=@n2), "Rotation Angle (0-1)"=Constant() {r=0.02}) {materialFunction=&CustomRotator}.Rotated Values) {r=true}
  LinearInterpolate(
   A=Constant3Vector() {constant=<rgba 0 0.0176 0.0782 1>}
   B=Constant3Vector() {constant=<rgba 0 0.0813 0.5312 1>}
   Alpha=Multiply(A=MaterialFunctionCall!(Value=@n3, Scale, Height) {materialFunction=/FortUILibrary/MaterialFunctions/Basics/MF_UI_PingPong}.Result) {constB=0.5}
  )
 ]
}
