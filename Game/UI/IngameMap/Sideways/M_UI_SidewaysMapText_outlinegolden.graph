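# Translucent UI material for the sideways in-game map, named as a golden outline for its text, that builds an animated swirling gold color.
# Two samples of one swirl flow texture pan at different tilings and speeds.
# Nested constant-alpha lerps, some extrapolating past 1, mix them with Light, Medium and Dark gold color parameters.
# The graph shows no material output links, and a Constant and OneMinus sit unconnected.
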Material /Game/UI/IngameMap/Sideways/M_UI_SidewaysMapText_outlinegolden
graph {
 properties {
  bUsedWithStaticMesh=false
  blendMode=BLEND_Translucent
  materialDomain=MD_UI
  pixelDepthOffsetMode=PDOM_Legacy
  usedShadingModels=Default Lit
 }
 nodes [
  Constant() {r=1}
  n1 = TextureSample(UVs=Panner(Coordinate=TextureCoordinate() {vTiling=-0.5}, Speed=Constant() {r=0.1}) {speedX=-0.06, speedY=0.025}) {texture=/Game/Global/Textures/Default/Flowmaps/T_SwirlFlow_02_D}
  n2 = TextureSample(UVs=Panner(Coordinate=TextureCoordinate() {uTiling=-0.5, vTiling=-1}, Speed=Constant() {r=0.08}) {speedX=-0.06, speedY=0.025}) {texture=&T_SwirlFlow_02_D}
  n3 = LinearInterpolate(
   A=VectorParameter() {parameterName=Light, defaultValue=<rgba 0.8125 0.5916 0 1>}.RGB
   B=VectorParameter() {parameterName=Medium, defaultValue=<rgba 0.1354 0.083 0 1>}.RGB
  ) {constAlpha=0.2}
  n4 = VectorParameter() {parameterName=Dark, defaultValue=<rgba 0.0833 0.0353 0 1>}
  LinearInterpolate(
   A=LinearInterpolate(A=LinearInterpolate(A=@n1.RGB, B=@n2.RGB) {constAlpha=2}, B=@n3)
   B=LinearInterpolate(A=Multiply(A=@n2.RGB) {constB=2}, B=LinearInterpolate(A=@n3, B=@n4.RGB) {constAlpha=1}) {constAlpha=1.4}
  ) {constAlpha=0.6, constB=1.5}
  OneMinus()
 ]
}
